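# Default Lit material named for red mistletoe in the main player's Take The Elf emote assets.
# Its graph holds a single vector parameter that defaults to a saturated red.
# No textures, other expressions or links to material outputs appear in the graph.
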Material /BRCosmetics/Animation/Game/MainPlayer/Emotes/Take_The_Elf/mistletoe_red
graph {properties {pixelDepthOffsetMode=PDOM_Legacy, usedShadingModels=Default Lit}, nodes [VectorParameter() {defaultValue=<rgba 1 0.0258 0 0>}]}
